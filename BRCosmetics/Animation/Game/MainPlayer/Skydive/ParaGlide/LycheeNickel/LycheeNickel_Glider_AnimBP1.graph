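# Animation Blueprint for the LycheeNickel glider, animating it through closed, opening and open-idle states from skydiving input or a force-open flag.
# States choose front-end preview clips or gameplay animations, with acceleration-driven blendspaces while opening and open.
# Curve-weighted additives layer aim lean and preview jitter, then curve-weighted Copy Bone and Leg IK attach both chains to the handles.
# The EventGraph derives acceleration and flap values and flags full deployment on notify.
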AnimBlueprint /BRCosmetics/Animation/Game/MainPlayer/Skydive/ParaGlide/LycheeNickel/LycheeNickel_Glider_AnimBP1
target_skeleton=/BRCosmetics/Gadgets/Assets/VinderTech_GliderChute/Glider_LycheeNickel/Meshes/LycheeNickel_Glider_Skeleton
graphs [
 AnimGraph {nodes [
   n1 = Copy Bone - Source Bone: Handle_Base_r - Target Bone: VB Root_Handle_Base_r!(
    ComponentPose:Component Space Pose Link Structure=Local To Component(LocalPose=Use cached pose 'Locomotion'())
    AlphaCurveName:Name=copybone
    bCopyTranslation:bool=true
    bCopyRotation:bool=true
    bCopyScale:bool=False
   ) -> (Pose:Component Space Pose Link Structure) {
    is=title AnimGraphNode_CopyBone
    alphaInputType=Curve
    alphaBoolBlend {blendOption=Linear}
    alphaScaleBiasClamp {bias=1, scale=-1}
    sourceBone {boneName=Handle_Base_r}
    targetBone {boneName=VB Root_Handle_Base_r}
   }
   n2 = Copy Bone - Source Bone: Handle_Base_l - Target Bone: VB Root_Handle_Base_l!(
    ComponentPose:Component Space Pose Link Structure=@n1
    AlphaCurveName:Name=copybone
    bCopyTranslation:bool=true
    bCopyRotation:bool=true
    bCopyScale:bool=False
   ) -> (Pose:Component Space Pose Link Structure) {
    is=title AnimGraphNode_CopyBone
    alphaInputType=Curve
    alphaBoolBlend {blendOption=Linear}
    alphaScaleBiasClamp {bias=1, scale=-1}
    sourceBone {boneName=Handle_Base_l}
    targetBone {boneName=VB Root_Handle_Base_l}
   }
   n3 = Leg IK!(ComponentPose:Component Space Pose Link Structure=@n2, AlphaCurveName:Name=copybone) -> (Pose:Component Space Pose Link Structure) {
    is=title AnimGraphNode_LegIK
    alphaInputType=Curve
    maxIterations=50
    alphaBoolBlend {blendOption=Linear}
    alphaScaleBiasClamp {bias=1, scale=-1}
    legsDefinition [
     {
      bEnableKneeTwistCorrection=false
      bEnableRotationLimit=false
      footBoneForwardAxis=Z
      hingeRotationAxis=None
      minRotationAngle=15
      numBonesInLimb=11
      twistOffsetCurveName=None
      fKFootBone {boneName=Chain_l_12_Ik}
      iKFootBone {boneName=VB Root_Handle_Base_l}
     }
     {
      bEnableKneeTwistCorrection=false
      bEnableRotationLimit=false
      footBoneForwardAxis=Z
      hingeRotationAxis=None
      minRotationAngle=15
      numBonesInLimb=11
      twistOffsetCurveName=None
      fKFootBone {boneName=Chain_r_12_Ik}
      iKFootBone {boneName=VB Root_Handle_Base_r}
     }
    ]
   }
   n4 = Copy Bone - Source Bone: Handle_r - Target Bone: Handle_Chain_r!(
    ComponentPose:Component Space Pose Link Structure=@n3
    AlphaCurveName:Name=copybone
    bCopyTranslation:bool=true
    bCopyRotation:bool=true
    bCopyScale:bool=true
   ) -> (Pose:Component Space Pose Link Structure) {
    is=title AnimGraphNode_CopyBone
    alphaInputType=Curve
    alphaBoolBlend {blendOption=Linear}
    alphaScaleBiasClamp {bias=1, scale=-1}
    sourceBone {boneName=Handle_r}
    targetBone {boneName=Handle_Chain_r}
   }
   n5 = Copy Bone - Source Bone: Handle_l - Target Bone: Handle_Chain_l!(
    ComponentPose:Component Space Pose Link Structure=@n4
    AlphaCurveName:Name=copybone
    bCopyTranslation:bool=true
    bCopyRotation:bool=true
    bCopyScale:bool=true
   ) -> (Pose:Component Space Pose Link Structure) {
    is=title AnimGraphNode_CopyBone
    alphaInputType=Curve
    alphaBoolBlend {blendOption=Linear}
    alphaScaleBiasClamp {bias=1, scale=-1}
    sourceBone {boneName=Handle_l}
    targetBone {boneName=Handle_Chain_l}
   }
   Output Pose(Result=Component To Local(ComponentPose=@n5)) {layerGroup=None}
   n6 = Break Fort Anim Input Skydiving!(FortAnimInput_Skydiving:"Fort Anim Input Skydiving Structure (by ref)"=Get SkydivingState() -> (SkydivingState:Fort Anim Input Skydiving Structure)) -> (
    ParaGilderRootOffset:Vector=~
    ParaGliderRootRotationOverride:Rotator=~
    bUseParaGlideRootModifier:bool=~
    bIsSkydivingFromLaunchPad:bool=~
    bIsSkydivingFromBus:bool=~
    bIsInVortex:bool=~
    bIsDiving:bool=~
    bIsParachuteOpen:bool=~
    bIsSkydiving:bool=~
    bIsParachuteLeaning:bool=~
    bIsSkydiveLeaning:bool=~
    bIsLeaning:bool=~
    bIsSkydiveDiveMode:bool=~
    bPlaySkydiveDrift:bool=~
    bSkydiveDriftDelayActive:bool=~
    bSkydiveDriftAnimAllowed:bool=~
    bIsGliderRight:bool=~
    bIsGliderCenter:bool=~
    bIsGliderLeft:bool=~
    bIsGliderForward:bool=~
    bIsGliderBack:bool=~
    bHasDeployedAndIsStandingUpright:bool=~
    LocalAccelForward:float=~
    LocalAccelRight:float=~
    LocalVelocityRight:float=~
    SkydiveAimPitch:float=~
    SkydiveAimYaw:float=~
    DeployChuteAnimRate:float=~
    SkydiveDriftAnimRate:float=~
    SkydiveDriftAnimRateCurrent:float=~
    SkydiveFidgetAnimRate:float=~
    SkydiveFidgetAnimRateCurrent:float=~
    SkydiveAdditiveAlpha:float=~
    SkydiveDriftDelay:float=~
    SkydiveDriftAnim:int
    SkydiveDriftAnimMax:int
    LaunchpadAnim:int
    LocalAccelDir:ESkydivingDirection Enum
    DirectionLast:ESkydivingDirection Enum
    PoseAdditive:Anim Sequence Object Reference
   ) {is=title K2Node_BreakStruct}
   n7 = Blendspace Player 'BS_LycheeNickel_GLIDER'(X=@n6.SkydiveAimYaw, Y=@n6.SkydiveAimPitch) {blendSpace=&/BS_LycheeNickel_GLIDER}
   AnimGraphNode_StateMachine_1 = Locomotion() {is=title AnimGraphNode_StateMachine}
   n8 = "Sequence Player (additive) 'LycheeNickel_Jitter_FrontEnd'"() {sequence=&/LycheeNickel_Jitter_FrontEnd}
   n9 = Get IsCosmeticPreview!(self:Fort Player Parachute Object Reference=Get OwnerGlider() -> (OwnerGlider:Fort Player Parachute Object Reference)) -> (IsCosmeticPreview:bool)
   Save cached pose 'Locomotion'(Pose=Apply Additive!(
     Base=Apply Additive(Base=@AnimGraphNode_StateMachine_1, Additive=@n7, AlphaCurveName=ParaGlide_Lean) {alphaInputType=Curve, alphaBoolBlend {blendOption=Linear}, alphaScaleBiasClamp {bias=1, scale=-1}}
     Additive=Blend Poses by bool(BlendPose_0=@n8, bActiveValue=@n9) {blendType=Linear}
     AlphaCurveName:Name=ParaGlide_Lean
    ) -> (Pose) {alphaInputType=Curve, alphaBoolBlend {blendOption=Linear}, alphaScaleBiasClamp {bias=1, scale=-1}}) {cacheName=Locomotion}
  ]}
 Locomotion {
  entry=Closed
  path=AnimGraph.AnimGraphNode_StateMachine_1.Locomotion
  ids=id
  nodes [
   AnimStateNode_0 = Idle_Open() {is=title AnimStateNode}
   AnimStateNode_1 = Open_Into() {is=title AnimStateNode}
   AnimStateNode_2 = Closed() {is=title AnimStateNode}
   AnimStateTransitionNode_0 = AnimStateTransitionNode() {bAutomaticRuleBasedOnSequencePlayerInState=true, blendMode=Linear, crossfadeDuration=0}
   AnimStateTransitionNode_1 = AnimStateTransitionNode() {crossfadeDuration=0}
   AnimStateTransitionNode_2 = AnimStateTransitionNode() {crossfadeDuration=0}
   AnimStateTransitionNode_3 = AnimStateTransitionNode() {crossfadeDuration=0, priorityOrder=2}
  ]
  links [
   {
    dst=Open_Into
    rule=AnimGraph.AnimGraphNode_StateMachine_1.Locomotion.AnimStateTransitionNode_1.Transition
    src=Closed
    via=AnimStateTransitionNode_1
   }
   {
    dst=Closed
    rule=AnimGraph.AnimGraphNode_StateMachine_1.Locomotion.AnimStateTransitionNode_2.Transition
    src=Idle_Open
    via=AnimStateTransitionNode_2
   }
   {
    dst=Closed
    rule=AnimGraph.AnimGraphNode_StateMachine_1.Locomotion.AnimStateTransitionNode_3.Transition
    src=Open_Into
    via=AnimStateTransitionNode_3
   }
   {
    dst=Idle_Open
    rule=AnimGraph.AnimGraphNode_StateMachine_1.Locomotion.AnimStateTransitionNode_0.Transition
    src=Open_Into
    via=AnimStateTransitionNode_0
   }
  ]
 }
 Idle_Open {
  path=AnimGraph.AnimGraphNode_StateMachine_1.Locomotion.AnimStateNode_0.Idle_Open
  nodes [
   n1 = Break Fort Anim Input Skydiving(FortAnimInput_Skydiving=Get SkydivingState()) -> (
    ParaGilderRootOffset=~
    ParaGliderRootRotationOverride=~
    bUseParaGlideRootModifier=~
    bIsSkydivingFromLaunchPad=~
    bIsSkydivingFromBus=~
    bIsInVortex=~
    bIsDiving=~
    bIsParachuteOpen=~
    bIsSkydiving=~
    bIsParachuteLeaning=~
    bIsSkydiveLeaning=~
    bIsLeaning=~
    bIsSkydiveDiveMode=~
    bPlaySkydiveDrift=~
    bSkydiveDriftDelayActive=~
    bSkydiveDriftAnimAllowed=~
    bIsGliderRight=~
    bIsGliderCenter=~
    bIsGliderLeft=~
    bIsGliderForward=~
    bIsGliderBack=~
    bHasDeployedAndIsStandingUpright=~
    LocalAccelForward=~
    LocalAccelRight=~
    LocalVelocityRight=~
    SkydiveAimPitch=~
    SkydiveAimYaw=~
    DeployChuteAnimRate=~
    SkydiveDriftAnimRate=~
    SkydiveDriftAnimRateCurrent=~
    SkydiveFidgetAnimRate=~
    SkydiveFidgetAnimRateCurrent=~
    SkydiveAdditiveAlpha=~
    SkydiveDriftDelay=~
   )
   n2 = Get OwnerGlider()
   Output Animation Pose(Result=Blend Poses by bool(
     BlendPose_0=Sequence Player 'LycheeNickel_Glide_FrontEnd_GLIDER'() {sequence=&/LycheeNickel_Glide_FrontEnd_GLIDER}
     BlendPose_1=Blendspace Player 'BS_LycheeNickel_Glide_GLIDER'(X=@n1.LocalAccelRight, Y=@n1.LocalAccelForward) {blendSpace=&/BS_LycheeNickel_Glide_GLIDER}
     bActiveValue=Get IsCosmeticPreview(self=@n2)
    ) {blendType=Linear})
  ]
 }
 Open_Into {
  path=AnimGraph.AnimGraphNode_StateMachine_1.Locomotion.AnimStateNode_1.Open_Into
  nodes [
   n1 = Break Fort Anim Input Skydiving(FortAnimInput_Skydiving=Get SkydivingState()) -> (
    ParaGilderRootOffset=~
    ParaGliderRootRotationOverride=~
    bUseParaGlideRootModifier=~
    bIsSkydivingFromLaunchPad=~
    bIsSkydivingFromBus=~
    bIsInVortex=~
    bIsDiving=~
    bIsParachuteOpen=~
    bIsSkydiving=~
    bIsParachuteLeaning=~
    bIsSkydiveLeaning=~
    bIsLeaning=~
    bIsSkydiveDiveMode=~
    bPlaySkydiveDrift=~
    bSkydiveDriftDelayActive=~
    bSkydiveDriftAnimAllowed=~
    bIsGliderRight=~
    bIsGliderCenter=~
    bIsGliderLeft=~
    bIsGliderForward=~
    bIsGliderBack=~
    bHasDeployedAndIsStandingUpright=~
    LocalAccelForward=~
    LocalAccelRight=~
    LocalVelocityRight=~
    SkydiveAimPitch=~
    SkydiveAimYaw=~
    DeployChuteAnimRate=~
    SkydiveDriftAnimRate=~
    SkydiveDriftAnimRateCurrent=~
    SkydiveFidgetAnimRate=~
    SkydiveFidgetAnimRateCurrent=~
    SkydiveAdditiveAlpha=~
    SkydiveDriftDelay=~
   )
   n2 = Sequence Player 'LycheeNickel_Into_FrontEnd_GLIDER'() {bLoopAnimation=false, sequence=&/LycheeNickel_Into_FrontEnd_GLIDER}
   n3 = Get OwnerGlider()
   Output Animation Pose(Result=Blend Poses by bool(
     BlendPose_0=@n2
     BlendPose_1=Blendspace Player 'BS_LycheeNickel_Into_GLIDER'(X=@n1.LocalAccelRight, Y=@n1.LocalAccelForward) {bLoop=false, blendSpace=&/BS_LycheeNickel_Into_GLIDER}
     bActiveValue=Get IsCosmeticPreview(self=@n3)
    ) {blendType=Linear})
  ]
 }
 Closed {
  path=AnimGraph.AnimGraphNode_StateMachine_1.Locomotion.AnimStateNode_2.Closed
  nodes [
   n1 = Sequence Player 'LycheeNickel_Closed_FrontEnd_GLIDER'() {bLoopAnimation=false, sequence=&/LycheeNickel_Closed_FrontEnd_GLIDER}
   n2 = Get OwnerGlider()
   Output Animation Pose(Result=Blend Poses by bool(
     BlendPose_0=@n1
     BlendPose_1=Sequence Player 'LycheeNickel_Closed_GLIDER'() {bLoopAnimation=false, sequence=&/LycheeNickel_Closed_GLIDER}
     bActiveValue=Get IsCosmeticPreview(self=@n2)
    ) {blendType=Linear})
  ]
 }
 Transition {path=AnimGraph.AnimGraphNode_StateMachine_1.Locomotion.AnimStateTransitionNode_0.Transition, nodes [Result()]}
 Transition {
  path=AnimGraph.AnimGraphNode_StateMachine_1.Locomotion.AnimStateTransitionNode_1.Transition
  nodes [
   n1 = Get SkydivingState()
   Result(bCanEnterTransition=OR Boolean(
     A=Break Fort Anim Input Skydiving(FortAnimInput_Skydiving=@n1) -> (
      ParaGilderRootOffset=~
      ParaGliderRootRotationOverride=~
      bUseParaGlideRootModifier=~
      bIsSkydivingFromLaunchPad=~
      bIsSkydivingFromBus=~
      bIsInVortex=~
      bIsDiving=~
      bIsParachuteOpen=~
      bIsSkydiving=~
      bIsParachuteLeaning=~
      bIsSkydiveLeaning=~
      bIsLeaning=~
      bIsSkydiveDiveMode=~
      bPlaySkydiveDrift=~
      bSkydiveDriftDelayActive=~
      bSkydiveDriftAnimAllowed=~
      bIsGliderRight=~
      bIsGliderCenter=~
      bIsGliderLeft=~
      bIsGliderForward=~
      bIsGliderBack=~
      bHasDeployedAndIsStandingUpright=~
      LocalAccelForward=~
      LocalAccelRight=~
      LocalVelocityRight=~
      SkydiveAimPitch=~
      SkydiveAimYaw=~
      DeployChuteAnimRate=~
      SkydiveDriftAnimRate=~
      SkydiveDriftAnimRateCurrent=~
      SkydiveFidgetAnimRate=~
      SkydiveFidgetAnimRateCurrent=~
      SkydiveAdditiveAlpha=~
      SkydiveDriftDelay=~
     ).bIsParachuteOpen
     B=Get bForceOpen() -> (bForceOpen:bool)
    ))
  ]
 }
 Transition {
  path=AnimGraph.AnimGraphNode_StateMachine_1.Locomotion.AnimStateTransitionNode_2.Transition
  nodes [
   n1 = Get bForceOpen()
   n2 = Break Fort Anim Input Skydiving(FortAnimInput_Skydiving=Get SkydivingState()) -> (
    ParaGilderRootOffset=~
    ParaGliderRootRotationOverride=~
    bUseParaGlideRootModifier=~
    bIsSkydivingFromLaunchPad=~
    bIsSkydivingFromBus=~
    bIsInVortex=~
    bIsDiving=~
    bIsParachuteOpen=~
    bIsSkydiving=~
    bIsParachuteLeaning=~
    bIsSkydiveLeaning=~
    bIsLeaning=~
    bIsSkydiveDiveMode=~
    bPlaySkydiveDrift=~
    bSkydiveDriftDelayActive=~
    bSkydiveDriftAnimAllowed=~
    bIsGliderRight=~
    bIsGliderCenter=~
    bIsGliderLeft=~
    bIsGliderForward=~
    bIsGliderBack=~
    bHasDeployedAndIsStandingUpright=~
    LocalAccelForward=~
    LocalAccelRight=~
    LocalVelocityRight=~
    SkydiveAimPitch=~
    SkydiveAimYaw=~
    DeployChuteAnimRate=~
    SkydiveDriftAnimRate=~
    SkydiveDriftAnimRateCurrent=~
    SkydiveFidgetAnimRate=~
    SkydiveFidgetAnimRateCurrent=~
    SkydiveAdditiveAlpha=~
    SkydiveDriftDelay=~
   )
   Result(bCanEnterTransition=AND Boolean(A=NOT Boolean(A=@n2.bIsParachuteOpen), B=NOT Boolean(A=@n1)))
  ]
 }
 Transition {
  path=AnimGraph.AnimGraphNode_StateMachine_1.Locomotion.AnimStateTransitionNode_3.Transition
  nodes [
   n1 = Get bForceOpen()
   n2 = Break Fort Anim Input Skydiving(FortAnimInput_Skydiving=Get SkydivingState()) -> (
    ParaGilderRootOffset=~
    ParaGliderRootRotationOverride=~
    bUseParaGlideRootModifier=~
    bIsSkydivingFromLaunchPad=~
    bIsSkydivingFromBus=~
    bIsInVortex=~
    bIsDiving=~
    bIsParachuteOpen=~
    bIsSkydiving=~
    bIsParachuteLeaning=~
    bIsSkydiveLeaning=~
    bIsLeaning=~
    bIsSkydiveDiveMode=~
    bPlaySkydiveDrift=~
    bSkydiveDriftDelayActive=~
    bSkydiveDriftAnimAllowed=~
    bIsGliderRight=~
    bIsGliderCenter=~
    bIsGliderLeft=~
    bIsGliderForward=~
    bIsGliderBack=~
    bHasDeployedAndIsStandingUpright=~
    LocalAccelForward=~
    LocalAccelRight=~
    LocalVelocityRight=~
    SkydiveAimPitch=~
    SkydiveAimYaw=~
    DeployChuteAnimRate=~
    SkydiveDriftAnimRate=~
    SkydiveDriftAnimRateCurrent=~
    SkydiveFidgetAnimRate=~
    SkydiveFidgetAnimRateCurrent=~
    SkydiveAdditiveAlpha=~
    SkydiveDriftDelay=~
   )
   Result(bCanEnterTransition=AND Boolean(A=NOT Boolean(A=@n2.bIsParachuteOpen), B=NOT Boolean(A=@n1)))
  ]
 }
 EventGraph {nodes [
   Delay(Duration=0.02)
   n1 = GetOwningComponent()
   > UnHideBoneByName!(execute, self:Skinned Mesh Component Object Reference=@n1, BoneName:Name=Glider_Base) -> (then)
   n2 = Event BlueprintUpdateAnimation()
   > Set Time(Time:float=@n2.DeltaTimeX) -> (Output_Get:float)
   > n3 = Sequence!(execute) -> (then_0, then_1, then_2:Exec, then_3:Exec)
   n4 = "float > float"(
    A=Break Fort Anim Input Skydiving(FortAnimInput_Skydiving=Get SkydivingState()) -> (
     ParaGilderRootOffset=~
     ParaGliderRootRotationOverride=~
     bUseParaGlideRootModifier=~
     bIsSkydivingFromLaunchPad=~
     bIsSkydivingFromBus=~
     bIsInVortex=~
     bIsDiving=~
     bIsParachuteOpen=~
     bIsSkydiving=~
     bIsParachuteLeaning=~
     bIsSkydiveLeaning=~
     bIsLeaning=~
     bIsSkydiveDiveMode=~
     bPlaySkydiveDrift=~
     bSkydiveDriftDelayActive=~
     bSkydiveDriftAnimAllowed=~
     bIsGliderRight=~
     bIsGliderCenter=~
     bIsGliderLeft=~
     bIsGliderForward=~
     bIsGliderBack=~
     bHasDeployedAndIsStandingUpright=~
     LocalAccelForward=~
     LocalAccelRight=~
     LocalVelocityRight=~
     SkydiveAimPitch=~
     SkydiveAimYaw=~
     DeployChuteAnimRate=~
     SkydiveDriftAnimRate=~
     SkydiveDriftAnimRateCurrent=~
     SkydiveFidgetAnimRate=~
     SkydiveFidgetAnimRateCurrent=~
     SkydiveAdditiveAlpha=~
     SkydiveDriftDelay=~
    ).LocalAccelForward
    B=1
   ) {is=title K2Node_CallFunction}
   Set FlappyRate(
    execute=@n3.then_3
    FlappyRate:double=FInterpTo(
     Current=Get FlappyRate() -> (FlappyRate:double)
     Target="To Float (Boolean)"(InBool=@n4)
     DeltaTime=Get Time() -> (Time:float)
     InterpSpeed=1
    )
   ) -> (Output_Get:double)
   n5 = Break Fort Anim Input Skydiving(FortAnimInput_Skydiving=Get SkydivingState()) -> (
    ParaGilderRootOffset=~
    ParaGliderRootRotationOverride=~
    bUseParaGlideRootModifier=~
    bIsSkydivingFromLaunchPad=~
    bIsSkydivingFromBus=~
    bIsInVortex=~
    bIsDiving=~
    bIsParachuteOpen=~
    bIsSkydiving=~
    bIsParachuteLeaning=~
    bIsSkydiveLeaning=~
    bIsLeaning=~
    bIsSkydiveDiveMode=~
    bPlaySkydiveDrift=~
    bSkydiveDriftDelayActive=~
    bSkydiveDriftAnimAllowed=~
    bIsGliderRight=~
    bIsGliderCenter=~
    bIsGliderLeft=~
    bIsGliderForward=~
    bIsGliderBack=~
    bHasDeployedAndIsStandingUpright=~
    LocalAccelForward=~
    LocalAccelRight=~
    LocalVelocityRight=~
    SkydiveAimPitch=~
    SkydiveAimYaw=~
    DeployChuteAnimRate=~
    SkydiveDriftAnimRate=~
    SkydiveDriftAnimRateCurrent=~
    SkydiveFidgetAnimRate=~
    SkydiveFidgetAnimRateCurrent=~
    SkydiveAdditiveAlpha=~
    SkydiveDriftDelay=~
   )
   n6 = NormalizeToRange!(Value:double=@n5.LocalAccelForward, RangeMin:double=0, RangeMax:double=2047) -> (ReturnValue:double)
   n7 = Branch(execute=@n3.then_0, Condition="float < float"(A=@n6))
   > n8 = MakeFloatPositive!(execute, self, Input:double=@n6) -> (then, Ouput:double)
   > Set AccelForward(AccelForward:double=@n8) -> (Output_Get:double)
   Set AccelForward(execute=@n7.else, AccelForward=@n6)
   n9 = NormalizeToRange(Value=@n5.LocalAccelRight, RangeMin=0, RangeMax=1499)
   n10 = Branch(execute=@n3.then_1, Condition="float < float"(A=@n9))
   > n11 = MakeFloatPositive(Input=@n9)
   > Set AccelBack(AccelBack:double=@n11) -> (Output_Get:double)
   Set AccelBack(execute=@n10.else, AccelBack=@n9)
   n12 = Get AccelForward() -> (AccelForward:double)
   n13 = Get AccelBack() -> (AccelBack:double)
   n14 = Branch(execute=@n3.then_2, Condition="float >= float"(A=@n12, B=@n13))
   > Set Acceleration(Acceleration:double=@n12) -> (Output_Get:double)
   Set Acceleration(execute=@n14.else, Acceleration=@n13)
   AnimNotify_TurnONEngines() {is=title K2Node_Event}
   > SetGliderFullyDeployed!(execute, self:Fort Glider Anim Instance Object Reference, bIsFullyDeployed:bool=true) -> (then)
   HideBoneByName!(execute, self:Skinned Mesh Component Object Reference=@n1, BoneName:Name=Glider_Base, PhysBodyOption:EPhysBodyOp Enum=PBO_None) -> (then)
   AnimNotify_ClosedEntered() {is=title K2Node_Event}
   AnimNotify_DeployEntered() {is=title K2Node_Event}
  ]}
 MakeFloatPositive {nodes [
   n1 = MakeFloatPositive!() -> (then, Input:double=~) {is=title K2Node_FunctionEntry}
   > n2 = Set float_input(float_input:double=float * float(A=@n1, B=-1) {is=title K2Node_PromotableOperator}) -> (Output_Get:double)
   > Return Node!(execute, Ouput:double=@n2)
  ]}
]
variables [
 AccelBack {type=number}
 Acceleration {type=number}
 AccelForward {type=number}
 FlappyRate {type=number}
 float_input {type=number}
 Time {type=number}
]
